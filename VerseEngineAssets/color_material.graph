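# UI-domain material in VerseEngineAssets that exposes a color parameter, apparently as a simple solid-color base for UI elements.
# It uses Alpha Composite (premultiplied alpha) blending and turns off automatic usage-flag setting in the editor.
# Its only node is a vector parameter named Color, defaulting to opaque white, with no link to a material output shown.
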Material /VerseEngineAssets/color_material
graph {
 properties {bAutomaticallySetUsageInEditor=false, blendMode=BLEND_AlphaComposite, materialDomain=MD_UI, usedShadingModels=Default Lit}
 nodes [VectorParameter() {parameterName=Color, defaultValue=<rgba 1 1 1 1>}]
}
